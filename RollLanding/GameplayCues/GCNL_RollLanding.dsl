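; Looping gameplay cue notify for roll landing that plays landing effects, force feedback and camera shake on the target.
; On application, it passes whether the target pawn is locally viewed to a Niagara component as IsLocalPlayer.
; On burst, it compares the cue's raw magnitude to a heavy fall height threshold, choosing heavy or light local force feedback and camera shake for the target's controller.
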(event EventBeginPlay)

(event Collision|EventActorBeginOverlap
  (params (OtherActor :type Actor%20Object%20Reference)))

(event EventTick
  (params (DeltaSeconds :type Float%20%28single-precision%29)))

(event EventOnApplication(Niagara)
  (params (MyTarget :type Actor%20Object%20Reference) (Parameters :type Gameplay%20Cue%20Parameters%20Structure%20%28by%20ref%29) (ParticleComponents :type Array%20of%20Niagara%20Particle%20System%20Component%20Object%20References) (AudioComponents :type Array%20of%20Audio%20Component%20Object%20References) (BurstCameraShakeInstance :type Camera%20Shake%20Base%20Object%20Reference) (BurstDecalInstance :type Decal%20Actor%20Object%20Reference))
  (|CallParentFunction MyTarget Parameters ParticleComponents AudioComponents BurstCameraShakeInstance BurstDecalInstance)
  (bind _asfort_player_pawn (Utilities|Casting|CastToFortPlayerPawn MyTarget))
  (Niagara|SetNiagaraVariable(Bool) (Utilities|Array|Get(acopy) ParticleComponents) "IsLocalPlayer" (Pawn|IsLocallyViewed _asfort_player_pawn)))

(fn OnBurst
  (params (MyTarget :type Actor%20Object%20Reference) (Parameters :type Gameplay%20Cue%20Parameters%20Structure%20%28by%20ref%29))
  (bind (_normalizedmagnitude _rawmagnitude) (Ability|GameplayCue|BreakGameplayCueParameters Parameters))
  (|CallParentFunction MyTarget Parameters)
  (LocalVariables|SetbIsHeavyFall (> _rawmagnitude (Variables|Default|GetHeavyFallHeightThreshold)))
  (Fortnite|PlayLocalForceFeedbackatLocation MyTarget (select (LocalVariables|GetbIsHeavyFall) (Variables|Default|GetHeavyForceFeedback) (Variables|Default|GetLightForceFeedback)) (Transformation|GetActorLocation MyTarget) 1.0)
  (Game|Feedback|ClientStartCameraShake :self (Fortnite|GetFortPlayerControllerfromActor MyTarget) :Shake (select (LocalVariables|GetbIsHeavyFall) (Variables|Default|GetHeavyCameraShake) (Variables|Default|GetLightCameraShake)) :Scale 1.0 :PlaySpace "CameraLocal")
  (return true))

(fn ConstructionScript)
